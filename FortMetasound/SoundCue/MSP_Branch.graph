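# MSP_Branch is a reusable MetaSound patch that works as a trigger branch on a Bool value.
# The In trigger fires a single Bool Trigger Compare node, with Value feeding its A operand and B left to a stored default.
# The node's True and False results pass straight out as the patch's two trigger outputs.
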MetaSoundPatch /FortMetasound/SoundCue/MSP_Branch
inputs [
 In {type=Trigger}
 Value {type=Bool}
]
outputs [
 False {type=Trigger}
 True {type=Trigger}
]
pages [Default {
  input_defaults {In=null, Value=null}
  nodes [TriggerCompareClamp_0 = TriggerCompare.Clamp:Bool(Compare=^In, A=^Value) {also {defaults {B="<MetaSoundBuilderResult.SUCCEEDED: 0>"}}}]
  links [{dst="@False", src=TriggerCompareClamp_0.False}, {dst="@True", src=TriggerCompareClamp_0.True}]
 }]
