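# MSP_ContinuousModulator is a MetaSound patch that remaps a continuous float from an input range to an output range.
# A single float Map Range node takes Value In with Input Min/Max and Output Min/Max, then sends the result to Value Out.
# Both ranges default to 0–1.
# Its SoundCue folder and name suggest it mirrors the Sound Cue Continuous Modulator node inside MetaSounds.
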MetaSoundPatch /FortMetasound/SoundCue/MSP_ContinuousModulator
inputs [
 Input.Max {type=Float}
 Input.Min {type=Float}
 Output.Max {type=Float}
 Output.Min {type=Float}
 Value In {type=Float}
]
outputs [Value Out {type=Float}]
pages [Default {
  input_defaults {Input.Max=1, Input.Min=0, Output.Max=1, Output.Min=0, Value In=0}
  nodes [MapRangeMapRange_0 = MapRange.MapRange:Float(In Range B=^Input.Max, In Range A=^Input.Min, Out Range B=^Output.Max, Out Range A=^Output.Min, In=^Value In)]
  links [{dst="@Value Out", src=MapRangeMapRange_0.Out Value}]
 }]
