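# MetaSound patch that generates a triggered fade value over Fade Time, with an option to invert it.
# In fires an AD envelope with zero attack and Fade Time as decay; Curve Exp drives only the attack curve.
# A trigger compare on bInvertOutput routes either the envelope or one minus it to Value Out.
# Out fires when the route is set, and On Done when the envelope finishes.
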MetaSoundPatch /FortMetasound/SoundCue/MSP_Fade
inputs [
 bInvertOutput {type=Bool}
 Curve Exp {type=Float}
 Fade Time {type=Float}
 In {type=Trigger}
]
outputs [
 On Done {type=Trigger}
 Out {type=Trigger}
 Value Out {type=Float}
]
pages [Default {
  input_defaults {Curve Exp=1, Fade Time=1, In=null, bInvertOutput=null}
  nodes [
   ADEnvelopeADEnvelope_0 = AD Envelope.AD Envelope:Audio(Attack Curve=^Curve Exp, Trigger=^In, Decay Time=UE.ConversionFloatToTime(In=^Fade Time).Out) {also {defaults {Attack Time=0}}}
   n1 = UE.ConversionAudioToFloat(In=@ADEnvelopeADEnvelope_0.Out Envelope)
   n2 = TriggerCompare.Clamp:Bool(Compare=^In, A=^bInvertOutput) {also {defaults {B="<MetaSoundBuilderResult.SUCCEEDED: 0>"}}}
   TriggerRouteTriggerRouteFloat2_0 = "TriggerRoute.Trigger Route (Float, 2):Float"(
    Value 1=@n1.Out
    Value 0=UE.Subtract:Float(AdditionalOperands=@n1.Out) {also {defaults {PrimaryOperand=1}}}.Out
    Set 0=@n2.False
    Set 1=@n2.True
   )
  ]
  links [
   {dst="@On Done", src=ADEnvelopeADEnvelope_0.On Done}
   {dst="@Out", src=TriggerRouteTriggerRouteFloat2_0.On Set}
   {dst="@Value Out", src=TriggerRouteTriggerRouteFloat2_0.Value}
  ]
 }]
