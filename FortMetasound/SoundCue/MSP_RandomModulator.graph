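# MetaSound patch that outputs a random volume and pitch value on each In trigger, with defaults close to unity.
# Two seeded random float nodes share the Seed input, and a trigger accumulator fires Out once both have produced a value.
# The pitch node takes Pitch.Min as both its bounds, so Pitch Out does not vary and the Pitch.Max input is unused.
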MetaSoundPatch /FortMetasound/SoundCue/MSP_RandomModulator
inputs [
 In {type=Trigger}
 Pitch.Max {type=Float}
 Pitch.Min {type=Float}
 Seed {type=Int32}
 Volume.Max {type=Float}
 Volume.Min {type=Float}
]
outputs [
 Out {type=Trigger}
 Pitch Out {type=Float}
 Volume Out {type=Float}
]
pages [Default {
  input_defaults {In=null, Pitch.Max=1.05, Pitch.Min=0.95, Seed=-1, Volume.Max=1.05, Volume.Min=0.95}
  nodes [
   RandomFloat_0 = UE.RandomFloat(Next=^In, Seed=^Seed, Max=^Volume.Max, Min=^Volume.Min)
   RandomFloat_1 = UE.RandomFloat(Next=^In, Max=^Pitch.Min, Min=^Pitch.Min, Seed=^Seed)
   TriggerAccumulatorTriggerAccumulate2_0 = "TriggerAccumulator.Trigger Accumulate (2)"(In 0=@RandomFloat_0.On Next, In 1=@RandomFloat_1.On Next)
  ]
  links [
   {dst="@Volume Out", src=RandomFloat_0.Value}
   {dst="@Pitch Out", src=RandomFloat_1.Value}
   {dst="@Out", src=TriggerAccumulatorTriggerAccumulate2_0.Out}
  ]
 }]
